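# Physics asset for the Gilliam vehicle body mesh, under the VehicleCosmetics bodies content.
# It holds a single body on the root bone, shaped as one box that is longer than it is wide or tall, offset mostly upward from the bone.
# There are no constraints, so the collision is one rigid box rather than a jointed or per-part setup.
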PhysicsAsset /VehicleCosmetics/Bodies/Gilliam/Meshes/Body_Gilliam_Physics_DM
bodies {root [{shapeType=Box, extentX=444, extentY=216, extentZ=116, shapeName=root_box, center=<6 0 102>}]}
constraints []
